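# Translucent, unlit, two-sided flame material for Niagara mesh particles on the repair tool blaster.
# Panning plasma and noise textures build the opacity mask, shaped by a vertical ramp, Fresnel and depth fade; HDR_Chromatic adds offset samples for RGB-split color.
# Particle color drives emissive, blending toward EndFlameColor along the ramp.
# Dynamic parameters control edge blend, brightness, fade and caustics-based distortion along vertex normals.
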Material /ResolveGameplay/Items/RepairTool/Assets/M_Blaster_FlameDance
graph {
 properties {
  allowTranslucentCustomDepthWrites=true
  bUsedWithNiagaraMeshParticles=true
  blendMode=BLEND_Translucent
  pixelDepthOffsetMode=PDOM_Legacy
  shadingModel=MSM_Unlit
  twoSided=true
  usedShadingModels=Unlit
 }
 nodes [
  Constant3Vector() {constant=<rgba 0 0 1 0>}
  Fresnel()
  n1 = Multiply(A=TextureCoordinate() {uTiling=6}, B=ScalarParameter() {defaultValue=1, parameterName=Tile_Tex_1})
  n2 = AppendVector(A=ScalarParameter() {parameterName=U_Offset}, B=ScalarParameter() {parameterName=V_Offset})
  n3 = Multiply(A=Time(), B=ScalarParameter() {defaultValue=1, parameterName=Movement Speed Multiplier})
  n4 = Panner(Coordinate=Add(A=@n1, B=@n2), Time=@n3) {speedX=0.125}
  n5 = Panner(Coordinate=@n1, Time=@n3) {speedX=0.125}
  n6 = Add(A=@n1, B=Multiply(A=@n2) {constB=-1})
  n7 = AppendVector(
   A=AppendVector(
    A=TextureSample(UVs=@n4) {texture=/Game/Effects/Fort_Effects/Textures/Tile/T_Plasma_BallEdge}.R
    B=TextureSample(UVs=@n5) {texture=&T_Plasma_BallEdge}.R
   )
   B=TextureSample(UVs=Panner(Coordinate=@n6, Time=@n3) {speedX=0.125}) {texture=&T_Plasma_BallEdge}.R
  )
  n8 = Panner(Coordinate=Multiply(A=TextureCoordinate() {uTiling=6}, B=ScalarParameter() {defaultValue=1, parameterName=Tile_Tex_1}), Time=@n3) {speedX=0.125}
  n9 = Multiply(A=TextureCoordinate() {vTiling=0.01}, B=ScalarParameter() {defaultValue=1, parameterName=Tile_Tex_0})
  n10 = AppendVector(A=ScalarParameter() {parameterName=U_Offset}, B=ScalarParameter() {parameterName=V_Offset})
  n11 = Add(A=@n9, B=@n10)
  n12 = AppendVector(
   A=TextureSample(UVs=Panner(Coordinate=@n11, Time=@n3) {speedX=0.012, speedY=-0.125}) {texture=/Game/Global/Textures/Default/Noise/T_TileNoise_02}.R
   B=TextureSample(UVs=Panner(Coordinate=@n9, Time=@n3) {speedX=0.012, speedY=-0.125}) {texture=&T_TileNoise_02}.R
  )
  n13 = Multiply(A=@n10) {constB=-1}
  n14 = TextureSample(UVs=Panner(Coordinate=Add(A=@n9, B=@n13), Time=@n3) {speedX=0.012, speedY=-0.125}) {texture=&T_TileNoise_02}
  n15 = Panner(Coordinate=Multiply(A=TextureCoordinate() {vTiling=0.01}, B=ScalarParameter() {defaultValue=1, parameterName=Tile_Tex_0}), Time=@n3) {speedX=0.012, speedY=-0.125}
  n16 = Multiply(A=TextureCoordinate() {uTiling=4, vTiling=0.025}, B=ScalarParameter() {defaultValue=1, parameterName=Tile_Tex_2})
  n17 = AppendVector(A=ScalarParameter() {parameterName=U_Offset}, B=ScalarParameter() {parameterName=V_Offset})
  n18 = Add(A=@n16, B=@n17)
  n19 = AppendVector(
   A=TextureSample(UVs=Panner(Coordinate=@n18, Time=@n3) {speedX=0.025, speedY=-0.1}) {texture=&T_TileNoise_02}.R
   B=TextureSample(UVs=Panner(Coordinate=@n16, Time=@n3) {speedX=0.025, speedY=-0.1}) {texture=&T_TileNoise_02}.R
  )
  n20 = Multiply(A=@n17) {constB=-1}
  n21 = TextureSample(UVs=Panner(Coordinate=Add(A=@n16, B=@n20), Time=@n3) {speedX=0.025, speedY=-0.1}) {texture=&T_TileNoise_02}
  n22 = Panner(
   Coordinate=Multiply(A=TextureCoordinate() {uTiling=4, vTiling=0.025}, B=ScalarParameter() {defaultValue=1, parameterName=Tile_Tex_2})
   Time=@n3
  ) {speedX=0.025, speedY=-0.1}
  n23 = Add(
   A=StaticSwitchParameter(True=AppendVector(A=@n12, B=@n14.R), False=TextureSample(UVs=@n15) {texture=&T_TileNoise_02}.R) {parameterName=HDR_Chromatic}
   B=StaticSwitchParameter(True=AppendVector(A=@n19, B=@n21.R), False=TextureSample(UVs=@n22) {texture=&T_TileNoise_02}.R) {parameterName=HDR_Chromatic}
  )
  n24 = Add(
   A=StaticSwitchParameter(True=@n7, False=TextureSample(UVs=@n8) {texture=&T_Plasma_BallEdge}.R) {parameterName=HDR_Chromatic}
   B=@n23
  )
  n25 = ParticleColor()
  n26 = Multiply(A=@n24, B=@n25.RGB)
  n27 = Clamp(None=@n26)
  n28 = Multiply(A=VectorParameter() {parameterName=MobileEmissiveColor, defaultValue=<rgba 0.974 0.895 0.6583 0>}.RGB, B=@n25.RGB)
  n29 = VectorParameter() {parameterName=EmissiveColor, defaultValue=<rgba 1 1 1 1>}
  n30 = StaticSwitchParameter(
   True=ShadingPathSwitch(
    Default=MaterialFunctionCall(LDR Input=@n27, HDR Tint=@n26) {materialFunction=&DeriveHDRfromLDR}.RemappedRange
    Mobile=@n28
   ) {desc=Migrated from FeatureLevelSwitch}
   False=ShadingPathSwitch(Default=Multiply(A=@n25.RGB, B=@n29.RGB), Mobile=@n28) {desc=Migrated from FeatureLevelSwitch}
  ) {parameterName=HDR_Chromatic}
  n31 = VectorParameter() {parameterName=EndFlameColor, defaultValue=<rgba 5 0.8721 0 0>}
  n32 = Constant() {r=1}
  n33 = DynamicParameter() {defaultValue=<rgba 0 0 0 1>, paramNames [LerpEdgeMask,BrightenMask,FadeDistance,DistortFlame]}
  n34 = Saturate(None=LinearInterpolate(
    A=Constant()
    B=Add(A=@n32, B=@n33.BrightenMask)
    Alpha=TextureSample() {texture=/Game/Global/Textures/Default/Masks_and_Gradients/T_Soft_Ramp_Vert_03}.R
   ))
  n35 = OneMinus(None=@n34)
  n36 = Power(Base=@n35)
  MaterialFunctionCall(Emissive=Multiply(A=LinearInterpolate(A=@n30, B=@n31.RGB, Alpha=@n36), B=ScalarParameter() {defaultValue=2, parameterName=MaskMult})) {materialFunction=&MF_FX_EmissiveAdjustments}
  n37 = ScalarParameter() {defaultValue=1.5, parameterName=Distort}
  Multiply(A=TextureSample() {texture=/Game/Global/Textures/Default/Masks_and_Gradients/CylGradient}.R, B=@n37) {constB=35}
  n38 = Multiply(A=TextureCoordinate(), B=ScalarParameter() {defaultValue=1, parameterName=TileTex})
  n39 = AppendVector(A=ScalarParameter() {parameterName=PannerSpeedX}, B=ScalarParameter() {defaultValue=-2, parameterName=PannerSpeedY})
  n40 = Multiply(
   A=TextureSample(UVs=Panner(Coordinate=@n38, Speed=@n39) {speedY=0.15}) {texture=&T_ShieldCaustics}.R
   B=VertexNormalWS()
  )
  Multiply(A=Multiply(A=Multiply(A=@n40, B=@n33.DistortFlame), B=@n35), B=@n37)
  n41 = MaterialFunctionCall(Axis=ObjectOrientation(), Exponent=ScalarParameter() {defaultValue=1, parameterName=AngleSens}) {materialFunction=&AxisAlignedFresnel}
  n42 = ScalarParameter() {defaultValue=8.5, parameterName=MultSens}
  n43 = LinearInterpolate(
   A=StaticSwitchParameter(True=Desaturation(None=@n23), False=@n23) {parameterName=HDR_Chromatic}
   B=StaticSwitchParameter(True=Desaturation(None=@n24), False=@n24) {parameterName=HDR_Chromatic}
   Alpha=@n33.LerpEdgeMask
  ) {constAlpha=1}
  n44 = Multiply(A=@n25.A, B=Multiply(A=ShadingPathSwitch(Default=@n43, Mobile=@n7) {desc=Migrated from FeatureLevelSwitch}, B=@n34))
  n45 = Fresnel(ExponentIn=ScalarParameter() {defaultValue=5, parameterName=FresnelExpo})
  n46 = StaticSwitchParameter(True=Multiply(A=@n44, B=OneMinus(None=@n45)), False=@n44) {parameterName=UseFresnelEdge}
  Multiply(
   A=Saturate(None=Multiply(A=@n41.Fresnel, B=@n42))
   B=QualitySwitch(Default=DepthFade(Opacity=@n46, FadeDistance=@n33.FadeDistance), Low=@n46)
  )
  Panner(Coordinate=Add(B=TextureCoordinate() {uTiling=0.5, vTiling=0.5})) {speedY=0.25}
 ]
}
